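# Physics asset for the Street Cuddles backpack accessory, setting up secondary motion for its two ears and gun holster.
# Capsule bodies on both ear joints and the holster are each constrained directly to the central pack capsule.
# Ear constraints have narrow twist and second-swing angles with a wider first swing, while the holster locks its first swing axis.
# A small sphere on the center-mass wiggle joint has no constraint in this asset.
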PhysicsAsset /Game/Accessories/FORT_Backpacks/Backpack_M_MED_Street_Cuddles/Meshes/M_MED_Street_Cuddles_Pack_Physics
bodies {
 dyn_L_ear_jnt [{shapeType=Capsule, length=3, radius=2, center=<2.36 -0.3479 0>, rotation=<rot 0 -90 0>}]
 dyn_R_ear_jnt [{shapeType=Capsule, length=3, radius=2, center=<2.4233 -0.3479 0>, rotation=<rot 0 -90 0>}]
 dyn_centerMassWiggle_jnt [{shapeType=Sphere, radius=0.7743}]
 dyn_gunHolster_jnt [{shapeType=Capsule, length=15.1387, radius=2, center=<1.187 0 -0.1138>, rotation=<rot 0 0 90>}]
 pack [{shapeType=Capsule, length=1, radius=4.1215}]
}
constraints [
 {bone1Name=dyn_gunHolster_jnt, bone2Name=pack, swing1LimitDegrees=45, swing1Motion=Locked, swing2LimitDegrees=30, twistLimitDegrees=2}
 {bone1Name=dyn_R_ear_jnt, bone2Name=pack, swing1LimitDegrees=30, swing2LimitDegrees=3, twistLimitDegrees=2}
 {bone1Name=dyn_L_ear_jnt, bone2Name=pack, swing1LimitDegrees=25, swing2LimitDegrees=3, twistLimitDegrees=3}
]
